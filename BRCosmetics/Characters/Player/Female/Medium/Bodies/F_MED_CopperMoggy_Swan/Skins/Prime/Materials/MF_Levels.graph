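# MF_Levels is a material function that applies a Photoshop-style input levels adjustment to a color, stored with the CopperMoggy Swan character skin's materials.
# It subtracts MinInput, floors at zero, divides by the MaxInput minus MinInput range and caps the result at one.
# That value is raised to the reciprocal of Gamma.
# There is no output range remap, and Color previews with the UV green-channel gradient.
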MaterialFunction /BRCosmetics/Characters/Player/Female/Medium/Bodies/F_MED_CopperMoggy_Swan/Skins/Prime/Materials/MF_Levels
graph {nodes [
  n1 = FunctionInput(Preview=ComponentMask(None=TextureCoordinate()) {g=true}) {bUsePreviewValueAsDefault=true, inputName=Color}
  n2 = FunctionInput() {bUsePreviewValueAsDefault=true, inputName=MinInput, inputType=FunctionInput_Scalar, sortPriority=1}
  n3 = FunctionInput() {bUsePreviewValueAsDefault=true, inputName=MaxInput, inputType=FunctionInput_Scalar, sortPriority=3, previewValue=<1 0 0 1>}
  n4 = Divide(A=Max(A=Subtract(A=@n1, B=@n2), B=Constant()), B=Subtract(A=@n3, B=@n2))
  n5 = Constant() {r=1}
  n6 = Constant() {r=1}
  n7 = FunctionInput() {bUsePreviewValueAsDefault=true, inputName=Gamma, inputType=FunctionInput_Scalar, sortPriority=2, previewValue=<1 0 0 1>}
  FunctionOutput(None=Power(Base=Min(A=@n4, B=@n5) {desc=LevelsControlInputRange}, Exp=Divide(A=@n6, B=@n7)) {desc=LevelsControlInput})
 ]}
